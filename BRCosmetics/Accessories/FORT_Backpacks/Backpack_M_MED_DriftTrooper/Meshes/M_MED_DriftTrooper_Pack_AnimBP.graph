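# Animation Blueprint for the DriftTrooper backpack that opens the pack when the player equips a melee weapon and closes it afterwards.
# A state machine plays idle, open and close sequences, with a separate open state for the front end.
# Each update derives the melee flag from the player's weapon animation type and maps spine yaw to roll and translation values.
# These values additively adjust the pack bone after a fixed offset.
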AnimBlueprint /BRCosmetics/Accessories/FORT_Backpacks/Backpack_M_MED_DriftTrooper/Meshes/M_MED_DriftTrooper_Pack_AnimBP
target_skeleton=/Game/Accessories/FORT_Backpacks/Backpack_M_MED_DriftTrooper/Meshes/M_MED_DriftTrooper_Pack_Skeleton
graphs [
 AnimGraph {nodes [
   Input Pose()
   AnimGraphNode_StateMachine_0 = PlayShieldAnim() {is=title AnimGraphNode_StateMachine}
   n1 = "Transform (Modify) Bone - Bone: pack"(ComponentPose=Local To Component(LocalPose=@AnimGraphNode_StateMachine_0), Translation="0,1.5,0", Rotation="0,0,0") {
    rotationMode=BMM_Additive
    rotationSpace=BCS_BoneSpace
    translationMode=BMM_Additive
    translationSpace=BCS_BoneSpace
    alphaBoolBlend {blendOption=Linear}
    boneToModify {boneName=pack}
    translation {y=1.5}
   }
   n2 = "Transform (Modify) Bone - Bone: pack"(ComponentPose=@n1, Translation=Get backpack_translation(), Rotation=Get backpack_rotate()) {rotationMode=BMM_Additive, translationMode=BMM_Additive, alphaBoolBlend {blendOption=Linear}, boneToModify {boneName=pack}}
   Output Pose(Result=Component To Local(ComponentPose=@n2)) {layerGroup=None}
  ]}
 PlayShieldAnim {
  entry=Idle
  path=AnimGraph.AnimGraphNode_StateMachine_0.PlayShieldAnim
  ids=id
  nodes [
   AnimStateNode_2 = Unequipped() {is=title AnimStateNode}
   AnimStateNode_4 = Idle() {
    stateEntered {guid=#1}
    stateFullyBlended {guid=#2}
    stateLeft {guid=#3}
   }
   AnimStateNode_5 = Equipped() {is=title AnimStateNode}
   AnimStateNode_6 = Equipped_FrontEnd() {is=title AnimStateNode}
   AnimStateTransitionNode_0 = AnimStateTransitionNode() {crossfadeDuration=0}
   AnimStateTransitionNode_1 = AnimStateTransitionNode() {crossfadeDuration=0}
   AnimStateTransitionNode_2 = AnimStateTransitionNode() {bAutomaticRuleBasedOnSequencePlayerInState=true, crossfadeDuration=0}
   AnimStateTransitionNode_4 = AnimStateTransitionNode()
  ]
  links [
   {
    dst=Unequipped
    rule=AnimGraph.AnimGraphNode_StateMachine_0.PlayShieldAnim.AnimStateTransitionNode_1.Transition
    src=Equipped
    via=AnimStateTransitionNode_1
   }
   {
    dst=Equipped
    rule=AnimGraph.AnimGraphNode_StateMachine_0.PlayShieldAnim.AnimStateTransitionNode_0.Transition
    src=Idle
    via=AnimStateTransitionNode_0
   }
   {
    dst=Equipped_FrontEnd
    rule=AnimGraph.AnimGraphNode_StateMachine_0.PlayShieldAnim.AnimStateTransitionNode_4.Transition
    src=Idle
    via=AnimStateTransitionNode_4
   }
   {
    dst=Idle
    rule=AnimGraph.AnimGraphNode_StateMachine_0.PlayShieldAnim.AnimStateTransitionNode_2.Transition
    src=Unequipped
    via=AnimStateTransitionNode_2
   }
  ]
 }
 Unequipped {
  path=AnimGraph.AnimGraphNode_StateMachine_0.PlayShieldAnim.AnimStateNode_2.Unequipped
  nodes [
   n1 = Sequence Player 'DriftTrooper_Close_Pack'!(bLoopAnimation:bool=False) -> (Pose) {bLoopAnimation=false, sequence=/BRCosmetics/Accessories/FORT_Backpacks/Backpack_M_MED_DriftTrooper/Animation/DriftTrooper_Close_Pack}
   Output Animation Pose(Result=@n1)
  ]
 }
 Idle {
  path=AnimGraph.AnimGraphNode_StateMachine_0.PlayShieldAnim.AnimStateNode_4.Idle
  nodes [
   n1 = Sequence Player 'DriftTrooper_Close_Idle_Pack'() {bLoopAnimation=false, sequence=/BRCosmetics/Accessories/FORT_Backpacks/Backpack_M_MED_DriftTrooper/Animation/DriftTrooper_Close_Idle_Pack}
   Output Animation Pose(Result=@n1)
  ]
 }
 Equipped {
  path=AnimGraph.AnimGraphNode_StateMachine_0.PlayShieldAnim.AnimStateNode_5.Equipped
  nodes [Output Animation Pose(Result=Sequence Player 'DriftTrooper_Open_Pack'!(bLoopAnimation:bool=False) -> (Pose) {bLoopAnimation=false, sequence=/BRCosmetics/Accessories/FORT_Backpacks/Backpack_M_MED_DriftTrooper/Animation/DriftTrooper_Open_Pack})]
 }
 Equipped_FrontEnd {
  path=AnimGraph.AnimGraphNode_StateMachine_0.PlayShieldAnim.AnimStateNode_6.Equipped_FrontEnd
  nodes [Output Animation Pose(Result=Sequence Player 'DriftTrooper_Open_Pack'(bLoopAnimation=False) {bLoopAnimation=false, sequence=&DriftTrooper_Open_Pack})]
 }
 Transition {
  path=AnimGraph.AnimGraphNode_StateMachine_0.PlayShieldAnim.AnimStateTransitionNode_0.Transition
  nodes [Result(bCanEnterTransition=Get PickaxeEquipped() -> (PickaxeEquipped:bool))]
 }
 Transition {
  path=AnimGraph.AnimGraphNode_StateMachine_0.PlayShieldAnim.AnimStateTransitionNode_1.Transition
  nodes [Result(bCanEnterTransition=NOT Boolean(A=Get PickaxeEquipped()))]
 }
 Transition {path=AnimGraph.AnimGraphNode_StateMachine_0.PlayShieldAnim.AnimStateTransitionNode_2.Transition, nodes [Result()]}
 Transition {
  path=AnimGraph.AnimGraphNode_StateMachine_0.PlayShieldAnim.AnimStateTransitionNode_4.Transition
  nodes [Result(bCanEnterTransition=Get bIsInFrontEnd())]
 }
 EventGraph {nodes [
   Event BlueprintUpdateAnimation()
   n1 = GetOwnerSkeletalMeshComponent()
   n2 = BreakRotator(InRot=BreakTransform(InTransform=GetDeltaTransformFromRefPose(self=@n1, BoneName=spine_05, BaseName=spine_01)).Rotation)
   > Set backpack_rotate(backpack_rotate=MakeRotator(Roll=MapRangeClamped(Value=@n2.Yaw, InRangeA=0, InRangeB=-20, OutRangeB=25)))
   > Set backpack_translation(backpack_translation=MakeVector(Y=MapRangeClamped(Value=@n2.Yaw, InRangeA=0, InRangeB=-20, OutRangeB=-2)))
   > Set PickaxeEquipped(PickaxeEquipped:bool=false) -> (Output_Get:bool)
   > n3 = Cast To FortPlayerAnimInstance(Object=GetAnimInstance(self=GetOwnerSkeletalMeshComponent()))
   > n4 = Switch on EFortWeaponCoreAnimation!(execute, Selection:EFortWeaponCoreAnimation Enum=Get WeaponCoreAnim(self=@n3)) -> (
    Melee:Exec
    Pistol:Exec
    Shotgun:Exec
    PaperBlueprint:Exec
    Rifle:Exec
    MeleeOneHand:Exec
    MachinePistol:Exec
    RocketLauncher:Exec
    GrenadeLauncher:Exec
    GoingCommando:Exec
    AssaultRifle:Exec
    TacticalShotgun:Exec
    SniperRifle:Exec
    TrapPlacement:Exec
    ShoulderLauncher:Exec
    AbilityDecoTool:Exec
    Crossbow:Exec
    C4:Exec
    RemoteControl:Exec
    DualWield:Exec
    AR_BullPup:Exec
    AR_ForwardGrip:Exec
    MedPackPaddles:Exec
    SMG_P90:Exec
    AR_DrumGun:Exec
    Consumable_Small:Exec
    Consumable_Large:Exec
    Balloon:Exec
    MountedTurret:Exec
    CreativeTool:Exec
    ExplosiveBow:Exec
    AshtonIndigo:Exec
    AshtonChicago:Exec
    MeleeDualWield:Exec
    Unarmed:Exec
   )
   Set PickaxeEquipped(execute=[@n4.Melee, @n4.MeleeDualWield, @n4.MeleeOneHand], PickaxeEquipped=true)
   TryGetPawnOwner()
  ]}
]
variables [
 As Fort Player Pawn {type=FortPlayerPawn}
 backpack_rotate {type=Rotator}
 backpack_translation {type=Vector}
 IsFrontEnd {type=boolean}
 PickaxeEquipped {type=boolean}
]
